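# UI material for social button icons, drawing an MSDF texture (microphone by default) with alpha-composite blending.
# Centered scale, rotation and offset adjust the UVs, and the MSDF converts to an SDF that is antialiased into an icon mask.
# A lower threshold on the same SDF yields a static-switched outline layered under the icon.
# Modular button state lerps supply color and opacity, and brightness and additive percentages tune the premultiplied output.
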Material /FortniteCoreUI/Social/Buttons/Materials/Icons/M_Social_Icon
graph {
 properties {blendMode=BLEND_AlphaComposite, materialDomain=MD_UI, pixelDepthOffsetMode=PDOM_Legacy, usedShadingModels=Default Lit}
 nodes [
  n1 = NamedRerouteUsage()
  n2 = NamedRerouteUsage()
  n3 = Multiply(A=ScalarParameter() {defaultValue=1, group=4- Alpha Composite, parameterName=Brightness Percentage, sliderMax=1, sortPriority=2}, B=@n2)
  n4 = NamedRerouteUsage()
  Multiply(A=Multiply(A=VertexColor().A, B=Multiply(A=@n1, B=@n3)), B=@n4)
  n5 = ScalarParameter() {group=4- Alpha Composite, parameterName=Additive Percentage, sliderMax=1, sortPriority=3}
  Multiply(A=@n4, B=Multiply(A=@n2, B=OneMinus(None=@n5)))
  n6 = MaterialFunctionCall!(Idle, Idle Opacity, Hovered, Hovered Opacity, Focused, Focused Opacity, Pressed, Pressed Opacity, Disabled, Disabled Opacity) {materialFunction=/FortUILibrary/Systems/ModularButton/MaterialFunctions/MF_UIKit_ButtonStatesBase_Lerp}
  n7 = MaterialFunctionCall!(
   Base States=@n6.Color
   Base States Opacity=@n6.Color
   Selected_Idle
   Selected_Idle Opacity
   Selected_Hovered
   Selected_Hovered Opacity
   Selected_Focused
   Selected_Focused Opacity
   Selected_Pressed
   Selected_Pressed Opacity
   Selected_Disabled
   Selected_Disabled Opacity
  ) {materialFunction=/FortUILibrary/Systems/ModularButton/MaterialFunctions/MF_UIKit_ButtonStatesSelectable_Lerp}
  NamedRerouteDeclaration(None=@n7.Color) {name=State Color, nodeColor=<rgba 0 0 0 1>}
  NamedRerouteDeclaration(None=@n7.Opacity) {name=State Opacity, nodeColor=<rgba 0 0 0 1>}
  n8 = LinearInterpolate(B=VectorParameter() {group=3 - Outline Settings, parameterName=Outline Color, defaultValue=<rgba 0 0 0 1>}.RGB, Alpha=NamedRerouteUsage())
  n9 = StaticSwitchParameter(True=@n8, False=Constant()) {group=3 - Outline Settings, parameterName=Display Outline}
  NamedRerouteDeclaration(None=LinearInterpolate(A=LinearInterpolate(B=@n8, Alpha=@n9), B=NamedRerouteUsage(), Alpha=NamedRerouteUsage())) {name=Colors, nodeColor=<rgba 1 0.7734 0 1>}
  n10 = TextureSampleParameter2D(UVs=NamedRerouteUsage()) {
   group=2 - Distance Field
   parameterName=Icon
   samplerType=SAMPLERTYPE_LinearColor
   texture=/FortniteCoreUI/Social/Buttons/Materials/Icons/Textures/T_UI_Icon_Social_Mic_MSDF
  }
  n11 = MaterialFunctionCall!(MSDF RGB=@n10.RGB) {materialFunction=/FortUILibrary/MaterialFunctions/Shapes/Modifiers/MF_UI_MSDFtoSDF}
  n12 = Constant() {r=0.5}
  NamedRerouteDeclaration(None=MaterialFunctionCall(DistanceField=@n11.SDF, Threshold=@n12, Smoothing=ScalarParameter() {group=2 - Distance Field, parameterName=Smoothing}) {materialFunction=&MF_UI_AntiAliasing}.Outer) {name=Icon Mask, nodeColor=<rgba 0.9688 0 1 1>}
  n13 = MaterialFunctionCall() {materialFunction=&GetUserInterfaceUV}
  n14 = AppendVector(
   A=Subtract(A=ComponentMask(None=@n13.Normalized UV) {r=true}, B=ScalarParameter() {group=1 - Transform, parameterName=OffsetX, sortPriority=48})
   B=Add(A=ComponentMask(None=@n13.Normalized UV) {g=true}, B=ScalarParameter() {group=1 - Transform, parameterName=OffsetY, sortPriority=48})
  )
  n15 = ScalarParameter() {defaultValue=1, group=1 - Transform, parameterName=Scale}
  NamedRerouteDeclaration(None=MaterialFunctionCall(
    UVs=MaterialFunctionCall(UVs=@n14, Texture Scale=@n15) {materialFunction=&ScaleUVsByCenter}.UVs
    "Rotation Angle (0-1)"=ScalarParameter() {group=1 - Transform, parameterName=Rotation}
   ) {materialFunction=&CustomRotator}.Rotated Values) {name=UVs, nodeColor=<rgba 0 0 0 1>}
  n16 = ScalarParameter() {defaultValue=0.35, group=3 - Outline Settings, parameterName=Outline Scale, sliderMax=1}
  NamedRerouteDeclaration(None=MaterialFunctionCall(
    DistanceField=@n11.SDF
    Threshold=Subtract(A=@n12, B=@n16) {constB=0.2}
    Smoothing=ScalarParameter() {group=2 - Distance Field, parameterName=Smoothing}
   ) {materialFunction=&MF_UI_AntiAliasing}.Outer) {name=Icon Outline Mask, nodeColor=<rgba 0 1 0.6484 1>}
  n17 = Multiply(
   A=NamedRerouteUsage()
   B=ScalarParameter() {defaultValue=0.25, group=3 - Outline Settings, parameterName=Outline Opacity, sliderMax=1, sortPriority=34}
  ) {constB=0.5}
  n18 = StaticSwitchParameter(True=@n17, False=Constant()) {group=3 - Outline Settings, parameterName=Display Outline}
  NamedRerouteDeclaration(None=LinearInterpolate(
    A=LinearInterpolate(B=@n17, Alpha=@n18)
    B=ScalarParameter() {defaultValue=1, group=1 - Transform, parameterName=Opacity, sliderMax=1}
    Alpha=NamedRerouteUsage()
   )) {name=Opacity, nodeColor=<rgba 0 0.6484 1 1>}
 ]
}
